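# Post-process material that renders a chromatic radial blur around a world-space center projected to the screen.
# It builds a distance-scaled, view-facing radial mask, optional panning radial noise distortion and a curve-atlas camera-distance fade.
# Custom HLSL accumulates dithered, color-weighted scene samples toward the center, with quality-capped sample count and zero fallback on lower feature levels.
# A static switch lets the MPC_PPM collection drive position, radius and distortion for Sequencer, while amplitude always reads from it.
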Material /DealershipSharedLibrary/FX/Materials/MasterMaterials/PostProcess/ChromaticBlur/PPM_ChromaticRadialBlur
graph {
 properties {
  bAutomaticallySetUsageInEditor=false
  bUsedWithNiagaraMeshParticles=true
  blendMode=BLEND_Translucent
  materialDomain=MD_PostProcess
  shadingModel=MSM_Unlit
  usedShadingModels=Unlit
 }
 nodes [
  n1 = MaterialFunctionCall(Float2=MaterialFunctionCall() {materialFunction=&ScreenResolution}.Visible Resolution) {materialFunction=&BreakOutFloat2Components}
  n2 = Divide(A=@n1.R, B=@n1.R)
  n3 = MaterialFunctionCall(Float2=MaterialFunctionCall(World Position=NamedRerouteUsage()) {materialFunction=&TransformToClipSpace}.Screen aligned 0-1 UV) {materialFunction=&BreakOutFloat2Components}
  n4 = Multiply(A=@n2, B=@n3.R)
  n5 = MaterialFunctionCall(Float2=ScreenPosition().ViewportUV) {materialFunction=&BreakOutFloat2Components}
  n6 = Multiply(A=@n2, B=@n5.R)
  n7 = Multiply(A=StaticSwitchParameter(
    True=CollectionParameter() {collection=/DealershipSharedLibrary/FX/Materials/MasterMaterials/PostProcess/MPC_PPM, parameterName=Radius}
    False=ScalarParameter() {defaultValue=100, group=Radial Mask, parameterName=Radial Mask - Radius, sortPriority=0}
   ) {group="!Master Controls", parameterName=Use Sequencer MPC, sortPriority=0}) {constB=0.64, desc=Converting to unit}
  n8 = Subtract(A=CameraPositionWS(), B=NamedRerouteUsage())
  n9 = Length(None=@n8) {is=class MaterialExpressionLength}
  n10 = Divide(A=@n7, B=@n9)
  n11 = Saturate(None=Divide(A=@n9, B=ScalarParameter() {group=Radial Mask, parameterName=Radial Mask - NearScaleSizeDistance}))
  n12 = OneMinus(None=DotProduct(A=@n8, B=MaterialFunctionCall() {materialFunction=&CameraDirectionVector}.Result))
  n13 = SphereMask(
   A=Subtract(A=AppendVector(A=@n4, B=@n3.G), B=AppendVector(A=@n6, B=@n5.G))
   B=Constant()
   Radius=Multiply(A=Multiply(A=@n10, B=@n11), B=Saturate(None=@n12))
   Hardness=ScalarParameter() {defaultValue=0.3, group=Radial Mask, parameterName=Radial Mask - Hardness, sortPriority=1}
  ) {attenuationRadius=1}
  NamedRerouteDeclaration(None=Multiply(A=@n13, B=@n13)) {name=Radial Mask, nodeColor=<rgba 0.0382 0.6328 1 1>}
  n14 = NamedRerouteUsage()
  n15 = MaterialFunctionCall!(
   UV In=Add(A=MaterialFunctionCall!(Center Point World Position=@n14) {materialFunction=/Engine/Functions/Engine_MaterialFunctions02/UVs/WorldSpaceAlignedScreenCoordinates}.Offset and Scaled UVS) {constB=0.5}
   Tile X=ScalarParameter() {defaultValue=0.01, group=Radial Noise, parameterName=Noise - UV Tile X, sortPriority=1}
   Tile Y=ScalarParameter() {defaultValue=1, group=Radial Noise, parameterName=Noise - UV Tile Y, sortPriority=2}
   Panning Speed X=ScalarParameter() {defaultValue=0.1, group=Radial Noise, parameterName=Noise - UV Panner X, sortPriority=3}
   Panning Speed Y=ScalarParameter() {defaultValue=0.01, group=Radial Noise, parameterName=Noise - UV Panner Y, sortPriority=4}
   Twist=ScalarParameter() {defaultValue=1, group=Radial Noise, parameterName=Noise - UV Exp X, sliderMax=5, sortPriority=5}
   Exp X=ScalarParameter() {group=Radial Noise, parameterName=Noise - UV Twist, sliderMax=1, sortPriority=6}
  ) {materialFunction=/DealershipSharedLibrary/FX/Materials/MaterialFunctions/UVs/MF_DLR_RadialUV}
  n16 = TextureSampleParameter2D!(UVs=@n15.Input UVs, "DDX(UVs)"=@n15.Input UVs, "DDY(UVs)"=@n15.Input UVs, Apply View MipBias) {
   group=Radial Noise
   mipValueMode=TMVM_Derivative
   parameterName=Noise - Texture
   sortPriority=-2
   texture=/Game/Global/Textures/Default/Noise/Noise
  }
  n17 = Multiply(
   A=Add(
    A=ChannelMaskParameter(None=@n16.RGB) {group=Radial Noise, parameterName=Noise - Texture Channel, sortPriority=0}.R
    B=ScalarParameter() {group=Radial Noise, parameterName=Noise - Bias, sortPriority=1}
   )
   B=StaticSwitchParameter(
    True=CollectionParameter() {collection=&MPC_PPM, parameterName=Distortion}
    False=ScalarParameter() {defaultValue=0.2, group=Radial Noise, parameterName=Noise - Distortion, sortPriority=-10}
   ) {group="!Master Controls", parameterName=Use Sequencer MPC, sortPriority=0}
  )
  NamedRerouteDeclaration(None=StaticSwitchParameter(True=OneMinus(None=@n17), False=Constant() {r=1}) {group="!Master Controls", parameterName=Use Chromatic DIsortion Noise, sortPriority=1}) {name=Chromatic Distortion Out, nodeColor=<rgba 1 0.9091 0 1>}
  n18 = ScalarParameter() {defaultValue=2500, group=Camera Fade, parameterName=Camera Fade -Min Distance}
  n19 = Subtract(A=Distance(A=NamedRerouteUsage(), B=CameraPositionWS()), B=@n18)
  n20 = Subtract(A=ScalarParameter() {defaultValue=45000, group=Camera Fade, parameterName=Camera Fade -Max Distance}, B=@n18)
  n21 = CurveAtlasRowParameter(CurveTime=Saturate(None=Divide(A=@n19, B=@n20))) {
   atlas=/DealershipSharedLibrary/FX/Textures/Curves/Curve_DepthFade_Atlas
   curve=/DealershipSharedLibrary/FX/Textures/Curves/Curve_DepthFade_Color
   parameterName=Camera Fade Curve
  }
  NamedRerouteDeclaration(None=@n21.R) {name=Camera Fade Curve, nodeColor=<rgba 1 0.0978 0.1854 1>}
  MaterialExpressionCustom_2 = Custom() {
   code="return 1;\r\n}\r\n\r\nfloat3 Lerp3Color (float Alpha){\r\n    \r\nfloat3 R = float3(0,0,1.5);\r\nfloat3 G = float3(0,0.5,0);\r\nfloat3 B = float3(1,0,0);\r\n//float Alpha;\r\nreturn lerp(   lerp(R,G,saturate(Alpha*2))   ,B,saturate(Alpha*2-1));\r\n"
   description=Lerp3ColorFunction
   outputType=CMOT_Float1
   showCode=true
   inputs []
  }
  MaterialExpressionSceneTexture_1 = SceneTexture() {sceneTextureId=PPI_PostProcessInput0}
  n22 = Constant() {r=9}
  n23 = Constant() {r=3}
  n24 = Constant() {r=6}
  n25 = Constant() {r=11}
  MaterialExpressionReroute_1 = Reroute(None=Clamp(
    None=ScalarParameter() {defaultValue=9, group=Chromatic Blur, parameterName=Chromatic Blur - Samples}
    Max=QualitySwitch(Default=@n22, Low=@n23, High=@n22, Medium=@n24, Epic=@n25)
   ) {maxDefault=11, minDefault=3})
  n26 = SceneTexelSize()
  n27 = CollectionParameter() {collection=&MPC_PPM, parameterName=Amplitude}
  n28 = NamedRerouteUsage()
  MaterialExpressionMultiply_17 = Multiply(A=Multiply(A=Multiply(A=@n26, B=@n27), B=Add(A=@n28)), B=NamedRerouteUsage())
  MaterialExpressionScreenPosition_0 = ScreenPosition()
  MaterialExpressionNamedRerouteUsage_19 = NamedRerouteUsage()
  MaterialExpressionScalarParameter_9 = ScalarParameter() {group=Chromatic Blur, parameterName=Chromatic Blur - Refraction}
  MaterialExpressionScalarParameter_7 = ScalarParameter() {defaultValue=0.5, group=Chromatic Blur, parameterName=Chromatic Blur - Dither}
  n29 = Custom!(
   Function=@MaterialExpressionCustom_2
   SceneColor=@MaterialExpressionSceneTexture_1.Color
   Steps=@MaterialExpressionReroute_1
   DistanceSum=@MaterialExpressionMultiply_17
   ScreenUV=@MaterialExpressionScreenPosition_0.ViewportUV
   ObjectClipPos=@MaterialExpressionNamedRerouteUsage_19
   Refraction=@MaterialExpressionScalarParameter_9
   ScreenPos=@MaterialExpressionScreenPosition_0.ViewportUV
   NoiseIntensity=@MaterialExpressionScalarParameter_7
  ) {
   code="float2 ScreenUVN = ViewportUVToBufferUV(ScreenUV)+Refraction;\r\nfloat2 Dir = ViewportUVToBufferUV(ObjectClipPos) - ScreenUVN ;\r\n\r\nint IntSteps = floor(Steps);\r\nfloat PDistance = DistanceSum/IntSteps;\r\n\r\nfloat random = InterleavedGradientNoise(ScreenPos.xy, float(View.StateFrameIndexMod8));\r\n\r\nfloat3 SumWeight=0;\r\nfloat3 SumColor=0;\r\n\r\nfor(int i = 0 ; i < IntSteps ; i++)\r\n    {\r\n        float alpha = saturate(i/(float(IntSteps-1)));\r\n        float3 ColorM = Lerp3Color(alpha);\r\n        float2 Pos = Dir*(lerp(1,random,NoiseIntensity) * PDistance*i) + ScreenUV;\r\n        Pos = clamp(Pos, float2(0.0, 0.0), float2(1.0, 1.0));\r\n        SumColor += SceneTextureLookup(ViewportUVToSceneTextureUV(Pos,14), 14, false)*ColorM;\r\n        SumWeight += ColorM;\r\n}\r\n\r\nreturn SumColor/SumWeight;"
   description=CARadialBlur
   showCode=true
   inputs [
    {
     inputName=Function
     input {expression=&self:MaterialExpressionCustom_2}
    }
    {
     inputName=SceneColor
     input {
      expression=&self:MaterialExpressionSceneTexture_1
      mask=1
      maskA=1
      maskB=1
      maskG=1
      maskR=1
     }
    }
    {
     inputName=Steps
     input {expression=&self:MaterialExpressionReroute_1}
    }
    {
     inputName=DistanceSum
     input {expression=&self:MaterialExpressionMultiply_17}
    }
    {
     inputName=ScreenUV
     input {expression=&self:MaterialExpressionScreenPosition_0}
    }
    {
     inputName=ObjectClipPos
     input {expression=&self:MaterialExpressionNamedRerouteUsage_19}
    }
    {
     inputName=Refraction
     input {expression=&self:MaterialExpressionScalarParameter_9}
    }
    {
     inputName=ScreenPos
     input {
      expression=&self:MaterialExpressionScreenPosition_0
      outputIndex=1
     }
    }
    {
     inputName=NoiseIntensity
     input {expression=&self:MaterialExpressionScalarParameter_7}
    }
   ]
  }
  n30 = Constant()
  NamedRerouteDeclaration(None=FeatureLevelSwitch(Default=@n29, ES2=@n30, ES3_1=@n30, SM4_REMOVED=@n30)) {name=Chromatic Blur Out, nodeColor=<rgba 0.2656 0 1 1>}
  n31 = StaticSwitchParameter(
   True=ComponentMask(None=CollectionParameter() {collection=&MPC_PPM, parameterName=Position}) {b=true, g=true, r=true}
   False=VectorParameter() {group=Position, parameterName=Position, defaultValue=<rgba 0 0 0 1>}.RGB
  ) {group="!Master Controls", parameterName=Use Sequencer MPC, sortPriority=0}
  NamedRerouteDeclaration(None=MaterialFunctionCall(World Position=NamedRerouteDeclaration(None=@n31) {name=Center Location World Space, nodeColor=<rgba 1 0.5156 0 1>}) {materialFunction=&TransformToClipSpace}.Buffer UVs for Scene Textures) {name=Center Location Clip Space, nodeColor=<rgba 1 0 0.5859 1>}
  NamedRerouteUsage()
  Reroute(None=@n7)
  Reroute(None=MaterialFunctionCall(Center Point World Position=NamedRerouteUsage()) {materialFunction=&WorldSpaceAlignedScreenCoordinates}.Offset and Scaled UVS)
 ]
}
